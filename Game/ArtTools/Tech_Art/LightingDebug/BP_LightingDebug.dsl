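; Lighting debug Blueprint that toggles engine show flags for light functions, direct lighting, skylight, motion blur and bloom.
; Each custom event uses a FlipFlop to pick the off or on ShowFlag console command, starting with off, and runs it on local player controller 0.
; ToggleDirectionalLight switches the DirectLighting flag rather than a specific light.
; BeginPlay, overlap, Tick and the construction script are empty, and nothing here calls the toggles.
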(event EventBeginPlay)

(event Collision|EventActorBeginOverlap
  (params (OtherActor :type Actor%20Object%20Reference)))

(event EventTick
  (params (DeltaSeconds :type Float%20%28single-precision%29)))

(event Custom|ToggleLightFunction
  (bind _isa (Utilities|FlowControl|FlipFlop)
    (:A
      _)
    (:B
      _))
  (Development|ExecuteConsoleCommand (Utilities|String|SelectString "ShowFlag.LightFunctions 0" "ShowFlag.LightFunctions 1" _isa) (Game|GetLocalPlayerControllerFromID 0)))

(event Custom|ToggleDirectionalLight
  (bind _isa (Utilities|FlowControl|FlipFlop)
    (:A
      _)
    (:B
      _))
  (Development|ExecuteConsoleCommand (Utilities|String|SelectString "ShowFlag.DirectLighting 0" "ShowFlag.DirectLighting 1" _isa) (Game|GetLocalPlayerControllerFromID 0)))

(event Custom|ToggleSkylight
  (bind _isa (Utilities|FlowControl|FlipFlop)
    (:A
      _)
    (:B
      _))
  (Development|ExecuteConsoleCommand (Utilities|String|SelectString "ShowFlag.SkyLighting 0" "ShowFlag.SkyLighting 1" _isa) (Game|GetLocalPlayerControllerFromID 0)))

(event Custom|ToggleMotionBlur
  (bind _isa (Utilities|FlowControl|FlipFlop)
    (:A
      _)
    (:B
      _))
  (Development|ExecuteConsoleCommand (Utilities|String|SelectString "ShowFlag.MotionBlur 0" "ShowFlag.MotionBlur 1" _isa) (Game|GetLocalPlayerControllerFromID 0)))

(event Custom|ToggleBloom
  (bind _isa (Utilities|FlowControl|FlipFlop)
    (:A
      _)
    (:B
      _))
  (Development|ExecuteConsoleCommand (Utilities|String|SelectString "ShowFlag.Bloom 0" "ShowFlag.Bloom 1" _isa) (Game|GetLocalPlayerControllerFromID 0)))

(fn ConstructionScript)
