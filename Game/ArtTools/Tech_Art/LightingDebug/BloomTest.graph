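# Unlit debug material for testing bloom, showing a row of stripes that ramp steeply in brightness into HDR range.
# U is tiled into repeating cells, and a SphereMask around each cell's floored-plus-half center cuts a stripe whose width follows the "size" parameter.
# The stripes are multiplied by a power-curved LinearGradient along U that blends from black to far above one, giving stepped intensities for comparing bloom response.
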Material /Game/ArtTools/Tech_Art/LightingDebug/BloomTest
graph {
 properties {bUsedWithNanite=true, pixelDepthOffsetMode=PDOM_Legacy, shadingModel=MSM_Unlit, usedShadingModels=Unlit}
 nodes [
  n1 = MaterialFunctionCall() {materialFunction=&LinearGradient}
  n2 = ComponentMask(None=TextureCoordinate() {uTiling=8}) {r=true}
  n3 = Floor(None=@n2)
  Multiply(
   A=LinearInterpolate(A=Constant(), B=Constant() {r=100}, Alpha=Power(Base=@n1.UGradient) {constExponent=3})
   B=SphereMask(A=@n2, B=Add(A=@n3) {constB=0.5}, Radius=ScalarParameter() {defaultValue=0.1, parameterName=size})
  )
 ]
}
