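# Post-process lighting debug material for showing 50% grey on pixels marked with custom stencil value 2.
# A stencil test limits the pass to pixels equal to that value.
# A lerp blends scene color toward 0.5 grey using saturated custom stencil as alpha.
# A separate Power node raises the same grey to 2.2, a gamma-to-linear conversion, without feeding the lerp.
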Material /Game/ArtTools/Tech_Art/LightingDebug/M_50Grey
graph {
 properties {
  bEnableStencilTest=true
  materialDomain=MD_PostProcess
  pixelDepthOffsetMode=PDOM_Legacy
  stencilCompare=MSC_Equal
  stencilRefValue=2
  usedShadingModels=Default Lit
 }
 nodes [
  n1 = Constant() {r=0.5}
  LinearInterpolate(
   A=SceneTexture() {sceneTextureId=PPI_PostProcessInput0}.Color
   B=@n1
   Alpha=Saturate(None=SceneTexture() {sceneTextureId=PPI_CustomStencil}.Color)
  )
  Power(Base=@n1) {constExponent=2.2}
 ]
}
