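# M_PureBlack is an unlit material in the lighting debug tech-art folder, named for a solid black surface.
# Its graph holds a single Constant node left at its default of zero, with no textures or parameters.
# The material is flagged for use with Nanite meshes and uses the legacy pixel depth offset mode.
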Material /Game/ArtTools/Tech_Art/LightingDebug/M_PureBlack
graph {properties {bUsedWithNanite=true, pixelDepthOffsetMode=PDOM_Legacy, shadingModel=MSM_Unlit, usedShadingModels=Unlit}, nodes [Constant()]}
